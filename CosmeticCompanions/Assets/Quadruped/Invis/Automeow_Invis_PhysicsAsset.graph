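# Physics asset for the invisible variant of the Automeow quadruped cosmetic companion.
# It defines a single sphere body on the pelvis bone as a coarse collision volume for the whole skeleton.
# It has no constraints, so it sets up no articulated ragdoll.
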PhysicsAsset /CosmeticCompanions/Assets/Quadruped/Invis/Automeow_Invis_PhysicsAsset
bodies {pelvis [{shapeType=Sphere, radius=45, shapeName=pelvis_sphere}]}
constraints []
