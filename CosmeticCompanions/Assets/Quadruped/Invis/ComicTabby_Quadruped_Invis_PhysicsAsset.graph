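# Physics asset for the invisible variant of the ComicTabby quadruped cosmetic companion.
# It holds a single sphere body on the root bone, raised along Z above the bone's origin.
# There are no constraints and no per-bone bodies, so the asset gives one coarse collision volume rather than an articulated ragdoll.
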PhysicsAsset /CosmeticCompanions/Assets/Quadruped/Invis/ComicTabby_Quadruped_Invis_PhysicsAsset
bodies {root [{shapeType=Sphere, radius=35, shapeName=root_sphere, center=<0 0 26.3654>}]}
constraints []
