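# Physics asset for the invisible FlourCut quadruped variant in the CosmeticCompanions content.
# It holds a single body on the root bone, a capsule raised above the bone's origin and rotated a quarter turn.
# There are no constraints, so the skeleton gets one simple collision shape rather than a jointed ragdoll.
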PhysicsAsset /CosmeticCompanions/Assets/Quadruped/Invis/FlourCut_Quadruped_Invis_PhysicsAsset
bodies {root [{shapeType=Capsule, length=20, radius=20, shapeName=root_capsule, center=<0 0 21.8144>, rotation=<rot 0 0 90>}]}
constraints []
