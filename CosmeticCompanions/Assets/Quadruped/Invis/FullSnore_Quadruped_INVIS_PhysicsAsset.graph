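# Physics asset for the invisible FullSnore quadruped cosmetic companion.
# It holds a single capsule body on the pelvis bone, turned on its side and slightly offset along the bone.
# It has no constraints, so it gives the skeleton one simple collision shape rather than an articulated ragdoll.
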PhysicsAsset /CosmeticCompanions/Assets/Quadruped/Invis/FullSnore_Quadruped_INVIS_PhysicsAsset
bodies {pelvis [{shapeType=Capsule, length=12.5, radius=37.5, shapeName=pelvis_capsule, center=<6.365 0 0>, rotation=<rot 90 0 0>}]}
constraints []
