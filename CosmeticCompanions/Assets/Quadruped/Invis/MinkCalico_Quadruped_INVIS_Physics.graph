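# Physics asset for the invisible variant of the MinkCalico quadruped in CosmeticCompanions.
# It holds two sphere bodies: a tiny one on pelvis and a large one on root, raised above the bone origin.
# A single constraint links pelvis to root with both swings and twist left free, so it does not limit rotation between them.
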PhysicsAsset /CosmeticCompanions/Assets/Quadruped/Invis/MinkCalico_Quadruped_INVIS_Physics
bodies {
 pelvis [{shapeType=Sphere, radius=0.505, shapeName=pelvis_sphere}]
 root [{shapeType=Sphere, radius=66.5, shapeName=root_sphere, center=<0 0 59.1876>}]
}
constraints [{
  bone1Name=pelvis
  bone2Name=root
  swing1LimitDegrees=45
  swing1Motion=Free
  swing2LimitDegrees=45
  swing2Motion=Free
  twistLimitDegrees=45
  twistMotion=Free
 }]
