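# Physics asset for the invisible small quadruped cosmetic companion, giving its skeleton a minimal collision shape.
# It holds a single body on the pelvis bone, a rotated capsule offset from the bone.
# With no other bodies and no constraints, it describes one rigid volume rather than an articulated ragdoll.
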PhysicsAsset /CosmeticCompanions/Assets/Quadruped/Invis/Quadruped_Small_Invis_Physics
bodies {pelvis [{shapeType=Capsule, length=20, radius=30, shapeName=pelvis_capsule, center=<10.6773 0.9352 0>, rotation=<rot 0 -93.8366 90>}]}
constraints []
