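# Physics asset for the invisible SniffRust quadruped cosmetic companion.
# It holds a single capsule body on the pelvis bone, offset and rotated relative to that bone, as a rough stand-in for the creature's shape.
# It has no other bodies and no constraints, so it provides a simple collision volume rather than an articulated ragdoll.
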PhysicsAsset /CosmeticCompanions/Assets/Quadruped/Invis/SniffRust_Quadruped_INVIS_PhysicsAsset
bodies {pelvis [{shapeType=Capsule, length=45, radius=40, shapeName=pelvis_capsule, center=<12.5549 0.6878 0>, rotation=<rot 0 -91.2829 90>}]}
constraints []
