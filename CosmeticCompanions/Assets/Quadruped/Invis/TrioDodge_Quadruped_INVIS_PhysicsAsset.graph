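# Physics asset for the invisible variant of the TrioDodge quadruped in CosmeticCompanions.
# It holds a single body on the root bone, a zero-length capsule that acts as a sphere offset slightly from the bone.
# There are no constraints, so it provides one rough collision volume with no articulated ragdoll.
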PhysicsAsset /CosmeticCompanions/Assets/Quadruped/Invis/TrioDodge_Quadruped_INVIS_PhysicsAsset
bodies {root [{shapeType=Capsule, length=0, radius=34.634, shapeName=root_capsule, center=<0 -6.2307 30.6585>}]}
constraints []
